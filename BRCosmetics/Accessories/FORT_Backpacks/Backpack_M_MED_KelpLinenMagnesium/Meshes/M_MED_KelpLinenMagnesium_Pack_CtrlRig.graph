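# Control Rig for the KelpLinenMagnesium backpack cosmetic that corrects the pack and sword-handle chain bones from body pose.
# Spine twist and roll, remapped through curves, weight opposing transform offsets on the pack.
# DBNO repositions chain bases; otherwise crouching adds Kalman-smoothed pack offsets from head rotation and pelvis height.
# Pelvis pitch adds another offset unless DBNO or parachuting, and the left chain base always gets a small nudge.
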ControlRigBlueprint /BRCosmetics/Accessories/FORT_Backpacks/Backpack_M_MED_KelpLinenMagnesium/Meshes/M_MED_KelpLinenMagnesium_Pack_CtrlRig
graphs [
 Multiply_1_ContainedGraph {nodes [
   n1 = RigVMFunctionEntryNode()
   Multiply = &RigVMFunction_MathDoubleMul(A=@n1.A, B=@n1.B)
   RigVMFunctionReturnNode(Result=&RigVMFunction_MathDoubleMul(B=@n1.C, A=@Multiply.Result).Result)
  ]}
 RigVMFunctionLibrary {links [], nodes []}
 RigVMFunction_Sequence_1_1_ContainedGraph {nodes [
   n1 = RigVMFunctionEntryNode()
   n2 = &RigVMFunction_Sequence(ExecuteContext=@n1.ExecuteContext)
   n3 = &RigVMFunction_Sequence(ExecuteContext=@n2.B)
   RigVMFunctionReturnNode(A=@n2.A, B=@n3.A, C=@n3.B)
  ]}
 RigVMModel {nodes [
   RigUnit_BeginExecution = &RigUnit_BeginExecution()
   RigVMFunction_Sequence_1 = RigVMAggregateNode(ExecuteContext=@RigUnit_BeginExecution.ExecutePin)
   n1 = RigVMVariableNode() {variable=spine_05}
   n2 = RigVMVariableNode() {variable=spine_01}
   n3 = RigVMDispatchNode(Struct=/Script/RigVM.RigVMFunction_MathQuaternionToRotator(Value=&RigVMFunction_MathTransformMakeRelative(Parent=@n1.Value, Global=@n2.Value).Local.Rotation) {is=struct RigVMUnitNode, template="Cast::Execute(in Value,out Result)"}.Result) {template="DISPATCH_RigVMDispatch_BreakStruct(in Struct,out Elements)"}
   RigVMFunction_AnimEvalRichCurve = &RigVMFunction_AnimEvalRichCurve(
    Curve="(EditorCurveData=(Keys=((InterpMode=RCIM_Linear,TangentMode=RCTM_Auto,TangentWeightMode=RCTWM_WeightedNone,Time=-1,Value=-1,ArriveTangent=0,ArriveTangentWeight=0,LeaveTangent=0,LeaveTangentWeight=0),(InterpMode=RCIM_Cubic,TangentMode=RCTM_Auto,TangentWeigh ...(truncated)"
    SourceMinimum=-1
    TargetMinimum=-1
    Value=&RigVMFunction_MathDoubleRemap(SourceMaximum=50, SourceMinimum=-50, TargetMaximum=1, TargetMinimum=-1, bClamp=False, Value=@n3.Elements.Yaw).Result
   )
   &RigUnit_ModifyTransforms(
    ItemToModify=((Item=(Type=Bone,Name="pack"),Transform=(Rotation=(X=0.342,Y=0,Z=0,W=-0.9397),Translation=(X=0,Y=-10.3541,Z=2.6278),Scale3D=(X=1,Y=1,Z=1))))
    WeightMinimum=0
    Weight=@RigVMFunction_AnimEvalRichCurve.Result
    ExecutePin=@RigVMFunction_Sequence_1.A
   )
   > &RigUnit_ModifyTransforms(
    ItemToModify=((Item=(Type=Bone,Name="pack"),Transform=(Rotation=(X=0.1736,Y=0,Z=0,W=0.9848),Translation=(X=0,Y=-1,Z=0),Scale3D=(X=1,Y=1,Z=1))))
    WeightMinimum=0
    Weight=&RigVMFunction_MathDoubleMul(B=-1, A=@RigVMFunction_AnimEvalRichCurve.Result).Result
   )
   n4 = &RigVMFunction_AnimEvalRichCurve(
    Curve=(EditorCurveData=(Keys=((),(Time=1,Value=1)),DefaultValue=340282346638528859811704183484516925440,PreInfinityExtrap=RCCE_Constant,PostInfinityExtrap=RCCE_Constant),ExternalCurve=None)
    SourceMinimum=-1
    TargetMinimum=-1
    Value=&RigVMFunction_MathDoubleRemap(SourceMaximum=45, SourceMinimum=-45, TargetMaximum=1, TargetMinimum=-1, bClamp=False, Value=@n3.Elements.Roll).Result
   )
   > &RigUnit_ModifyTransforms(
    ItemToModify=((Item=(Type=Bone,Name="pack"),Transform=(Rotation=(X=-0.0872,Y=0,Z=0,W=0.9962),Translation=(X=0,Y=0,Z=0),Scale3D=(X=1,Y=1,Z=1))))
    WeightMinimum=0
    Weight=@n4.Result
   )
   > n5 = &RigUnit_ModifyTransforms(
    ItemToModify=((Item=(Type=Bone,Name="pack"),Transform=(Rotation=(X=0,Y=0,Z=0,W=1),Translation=(X=0,Y=0,Z=0),Scale3D=(X=1,Y=1,Z=1))))
    WeightMinimum=0
    Weight=&RigVMFunction_MathDoubleMul(B=-1, A=@n4.Result).Result
   )
   n6 = &RigVMFunction_ControlFlowBranch(ExecuteContext=@n5.ExecutePin, Condition=RigVMVariableNode() {variable=DBNO}.Value)
   n7 = &RigVMFunction_ControlFlowBranch(ExecuteContext=@n6.Completed, Condition=RigVMVariableNode() {variable=is Crouch}.Value)
   n8 = &RigVMFunction_MathTransformMakeRelative(Parent=RigVMVariableNode() {variable=pelvis}.Value, Global=RigVMVariableNode() {variable=head}.Value)
   n9 = &RigVMFunction_MathDoubleRemap(
    SourceMaximum=-30
    SourceMinimum=0
    TargetMaximum=1
    TargetMinimum=0
    Value=RigVMDispatchNode(Struct=&RigVMFunction_MathQuaternionToRotator(Value=@n8.Local.Rotation) {template="Cast::Execute(in Value,out Result)"}.Result) {template="DISPATCH_RigVMDispatch_BreakStruct(in Struct,out Elements)"}.Elements.Yaw
   )
   n10 = &RigVMFunction_MathTransformMakeRelative(
    Global=(Rotation=(X=0,Y=0,Z=0,W=1),Translation=(X=0,Y=0,Z=0),Scale3D=(X=1,Y=1,Z=1))
    Parent=(Rotation=(X=0,Y=0,Z=0,W=1),Translation=(X=0,Y=0,Z=0),Scale3D=(X=1,Y=1,Z=1))
    Global.Translation=RigVMVariableNode() {variable=pelvis}.Value.Translation
    Parent.Translation=RigVMVariableNode() {variable=root}.Value.Translation
   )
   n11 = &RigVMFunction_MathDoubleRemap(SourceMaximum=75, SourceMinimum=80, TargetMaximum=1, TargetMinimum=0, Value=@n10.Local.Translation.Z)
   &RigUnit_ModifyTransforms(
    ItemToModify=((Item=(Type=Bone,Name="pack"),Transform=(Rotation=(X=0.0262,Y=0,Z=0,W=0.9997),Translation=(X=0,Y=-1,Z=-5),Scale3D=(X=1,Y=1,Z=0.9))))
    WeightMinimum=0
    Weight=&RigVMFunction_KalmanFloat(BufferSize=8, Value=&RigVMFunction_MathDoubleMul(A=@n9.Result, B=@n11.Result).Result) {template="KalmanFilter::Execute(in Value,in BufferSize,out Result)"}.Result
    ExecutePin=@n7.True
   )
   Multiply_1 = RigVMAggregateNode(
    B=@n9.Result
    A=@n11.Result
    C=&RigVMFunction_MathDoubleRemap(SourceMaximum=65, SourceMinimum=60, TargetMaximum=1, TargetMinimum=0, Value=@n10.Local.Translation.Z).Result
   )
   > &RigUnit_ModifyTransforms(
    ItemToModify=((Item=(Type=Bone,Name="pack"),Transform=(Rotation=(X=0.061,Y=0,Z=0,W=0.9981),Translation=(X=0,Y=-1.2,Z=-6.6),Scale3D=(X=1,Y=1,Z=1))))
    WeightMinimum=0
    Weight=&RigVMFunction_KalmanFloat(BufferSize=16, Value=@Multiply_1.Result) {template="KalmanFilter::Execute(in Value,in BufferSize,out Result)"}.Result
   )
   &RigUnit_ModifyTransforms(
    ItemToModify="((Item=(Type=Bone,Name=\"base_chain_l\"),Transform=(Rotation=(X=0.2685,Y=0.2486,Z=-0.5477,W=-0.7524),Translation=(X=0.4,Y=-0.5,Z=0.2),Scale3D=(X=1,Y=1,Z=1))),(Item=(Type=Bone,Name=\"base_chain_r\"),Transform=(Rotation=(X=-0.154,Y=0.0802,Z=-0.4547,W=0.8735),Tra ...(truncated)"
    Weight=1
    WeightMinimum=0
    ExecutePin=@n6.True
   )
   &RigUnit_ModifyTransforms(
    ItemToModify=((Item=(Type=Bone,Name="base_chain_l"),Transform=(Rotation=(X=-0.0349,Y=-0.0349,Z=-0.0012,W=0.9988),Translation=(X=0.5,Y=0,Z=0.014),Scale3D=(X=1,Y=1,Z=1))))
    Weight=1
    WeightMinimum=0
    ExecutePin=@RigVMFunction_Sequence_1.C
   )
   n12 = RigVMVariableNode() {variable=is parachute open}
   n13 = RigVMVariableNode() {variable=DBNO}
   RigVMFunction_MathBoolOr = &RigVMFunction_MathBoolOr(B=@n12.Value, A=@n13.Value)
   n14 = &RigVMFunction_ControlFlowBranch(Condition=@RigVMFunction_MathBoolOr.Result, ExecuteContext=@RigVMFunction_Sequence_1.B)
   n15 = RigVMVariableNode() {variable=root}
   n16 = RigVMVariableNode() {variable=pelvis}
   n17 = RigVMDispatchNode(Struct=&RigVMFunction_MathQuaternionToRotator(Value=&RigVMFunction_MathTransformMakeRelative(Parent=@n15.Value, Global=@n16.Value).Local.Rotation) {template="Cast::Execute(in Value,out Result)"}.Result) {template="DISPATCH_RigVMDispatch_BreakStruct(in Struct,out Elements)"}
   &RigUnit_ModifyTransforms(
    ItemToModify=((Item=(Type=Bone,Name="pack"),Transform=(Rotation=(X=-0.0872,Y=0,Z=0,W=0.9962),Translation=(X=0,Y=-4,Z=-3),Scale3D=(X=1,Y=1,Z=1))))
    WeightMinimum=0
    Weight=&RigVMFunction_MathDoubleMul(
     A=&RigVMFunction_MathDoubleRemap(SourceMaximum=30, SourceMinimum=20, TargetMaximum=1, TargetMinimum=0, Value=@n17.Elements.Pitch).Result
     B=&RigVMFunction_MathDoubleRemap(SourceMaximum=40, SourceMinimum=50, TargetMaximum=1, TargetMinimum=0, Value=@n17.Elements.Pitch).Result
    ).Result
    ExecutePin=@n14.False
   )
   &RigVMFunction_AnimEvalRichCurve(
    Curve="(EditorCurveData=(Keys=((InterpMode=RCIM_Linear,TangentMode=RCTM_Auto,TangentWeightMode=RCTWM_WeightedNone,Time=0,Value=0,ArriveTangent=0,ArriveTangentWeight=0,LeaveTangent=0,LeaveTangentWeight=0),(InterpMode=RCIM_Cubic,TangentMode=RCTM_Auto,TangentWeightM ...(truncated)"
    SourceMaximum=50
    SourceMinimum=-15
    TargetMinimum=0
    Value=0
   )
   RigVMFunction_MathBoolNot = &RigVMFunction_MathBoolNot(Value=@n12.Value)
   &RigVMFunction_MathBoolNot(Value=@n13.Value)
  ]}
]
hierarchy [
 root {loc=<0 0 0>, rot=<0 0 0 1>}
 pack {parent=root, loc=<0 -0.7316 0>, rot=<0 0 0 1>}
 Sword_handle_l {parent=pack, loc=<5.9971 -1.9418 10.5502>, rot=<0 0 0 1>}
 base_chain_l {parent=Sword_handle_l, loc=<0 0 14.6223>, rot=<0 0 -0.0085 -1>}
 dyn_chain_l_1 {parent=base_chain_l, loc=<0 0 0>, rot=<0 0 0 1>}
 dyn_chain_l_2 {parent=dyn_chain_l_1, loc=<2.0841 -0.0353 0>, rot=<0 -0.4596 0 -0.8881>}
 dyn_chain_l_3 {parent=dyn_chain_l_2, loc=<1.5878 0 0>, rot=<0 -0.0816 0 -0.9967>}
 dyn_chain_l_4 {parent=dyn_chain_l_3, loc=<1.7594 0.121 -0.0387>, rot=<0.165 -0.2147 -0.0368 -0.9619>}
 dyn_chain_l_5 {parent=dyn_chain_l_4, loc=<2.0615 -0.1284 -0.1853>, rot=<0 -0.0211 0 -0.9998>}
 dyn_chain_l_6 {parent=dyn_chain_l_5, loc=<1.461 0 0>, rot=<0 0 0 1>}
 Sword_handle_r {parent=pack, loc=<-5.9971 -1.9418 10.5502>, rot=<0 0 0 1>}
 base_chain_r {parent=Sword_handle_r, loc=<0 0 14.6223>, rot=<-0.0003 0.0412 -0.0085 -0.9991>}
 dyn_chain_r_1 {parent=base_chain_r, loc=<0 0 0>, rot=<0 0 0 1>}
 dyn_chain_r_2 {parent=dyn_chain_r_1, loc=<-1.776 0.0353 0.2179>, rot=<0 -0.4735 0 -0.8808>}
 dyn_chain_r_3 {parent=dyn_chain_r_2, loc=<0.5291 0.0311 -1.4968>, rot=<0 0.0516 0 -0.9987>}
 dyn_chain_r_4 {parent=dyn_chain_r_3, loc=<0.4438 0.1543 -1.5571>, rot=<-0.0776 -0.2785 0.0842 -0.9536>}
 dyn_chain_r_5 {parent=dyn_chain_r_4, loc=<1.6469 0.0332 -0.6047>, rot=<0 -0.07 0 -0.9975>}
 dyn_chain_r_6 {parent=dyn_chain_r_5, loc=<1.7922 0.137 -0.0448>, rot=<0 0 0 1>}
]
variables [
 DBNO {type=bool, default=False}
 Melee {type=bool, default=False}
 head {type=FTransform, default=(Rotation=(X=0,Y=0,Z=0,W=1),Translation=(X=0,Y=0,Z=0),Scale3D=(X=1,Y=1,Z=1))}
 is Crouch {type=bool, default=False}
 is Skydive {type=bool, default=False}
 is parachute open {type=bool, default=False}
 jump {type=bool, default=False}
 neck_01 {type=FTransform, default=(Rotation=(X=0,Y=0,Z=0,W=1),Translation=(X=0,Y=0,Z=0),Scale3D=(X=1,Y=1,Z=1))}
 pelvis {type=FTransform, default=(Rotation=(X=0,Y=0,Z=0,W=1),Translation=(X=0,Y=0,Z=0),Scale3D=(X=1,Y=1,Z=1))}
 root {type=FTransform, default=(Rotation=(X=0,Y=0,Z=0,W=1),Translation=(X=0,Y=0,Z=0),Scale3D=(X=1,Y=1,Z=1))}
 spine_01 {type=FTransform, default=(Rotation=(X=0,Y=0,Z=0,W=1),Translation=(X=0,Y=0,Z=0),Scale3D=(X=1,Y=1,Z=1))}
 spine_03 {type=FTransform, default=(Rotation=(X=0,Y=0,Z=0,W=1),Translation=(X=0,Y=0,Z=0),Scale3D=(X=1,Y=1,Z=1))}
 spine_04 {type=FTransform, default=(Rotation=(X=0,Y=0,Z=0,W=1),Translation=(X=0,Y=0,Z=0),Scale3D=(X=1,Y=1,Z=1))}
 spine_05 {type=FTransform, default=(Rotation=(X=0,Y=0,Z=0,W=1),Translation=(X=0,Y=0,Z=0),Scale3D=(X=1,Y=1,Z=1))}
]
