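# Physics asset for the KelpLinenMagnesium backpack cosmetic, giving its pack, left and right sword handles and their two dangling chains capsule bodies for physics.
# Each side runs pack, sword handle, chain base, then six dynamic chain links, joined by constraints in series.
# Handle-to-pack joints have free swing and twist, while chain joints carry narrow angular limits.
# The two chains differ slightly in capsule sizes, offsets and limits, so they are not exact mirrors.
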PhysicsAsset /BRCosmetics/Accessories/FORT_Backpacks/Backpack_M_MED_KelpLinenMagnesium/Meshes/M_MED_KelpLinenMagnesium_Pack_Physics
bodies {
 Sword_handle_l [{shapeType=Capsule, length=8.3352, radius=1.6748, shapeName=Sword_handle_l_capsule, center=<0 0 8.2817>, rotation=<rot 0 -90 0>}]
 Sword_handle_r [{shapeType=Capsule, length=9.3069, radius=1.7031, shapeName=Sword_handle_l_capsule, center=<0 0 7.8776>, rotation=<rot 0 -90 180>}]
 base_chain_l [{shapeType=Capsule, length=0.505, radius=0.505, shapeName=base_chain_l_capsule, rotation=<rot 0 -90 0>}]
 base_chain_r [{shapeType=Capsule, length=0.505, radius=0.505, shapeName=base_chain_l_capsule, rotation=<rot 0 -90 180>}]
 dyn_chain_l_1 [{shapeType=Capsule, length=0.505, radius=1.5, shapeName=dyn_chain_l_1_capsule, rotation=<rot 0 90 -90>}]
 dyn_chain_l_2 [{shapeType=Capsule, length=1.4234, radius=1.5, shapeName=dyn_chain_l_2_capsule, rotation=<rot 0 90 -90>}]
 dyn_chain_l_3 [{shapeType=Capsule, length=0.597, radius=1.5, shapeName=dyn_chain_l_3_capsule, rotation=<rot 0 90 -90>}]
 dyn_chain_l_4 [{shapeType=Capsule, length=1.9128, radius=1.5, shapeName=dyn_chain_l_4_capsule, rotation=<rot 0 90 -90>}]
 dyn_chain_l_5 [{shapeType=Capsule, length=1.4234, radius=1.5, shapeName=dyn_chain_l_5_capsule, rotation=<rot 0 90 -90>}]
 dyn_chain_l_6 [{shapeType=Capsule, length=1.0522, radius=1.5, shapeName=dyn_chain_l_6_capsule, rotation=<rot 0 90 -90>}]
 dyn_chain_r_1 [{shapeType=Capsule, length=0.505, radius=1, shapeName=dyn_chain_l_1_capsule, rotation=<rot 0 89.0289 90.8035>}]
 dyn_chain_r_2 [{shapeType=Capsule, length=1.4234, radius=1, shapeName=dyn_chain_l_2_capsule, rotation=<rot 0 89.4179 91.5108>}]
 dyn_chain_r_3 [{shapeType=Capsule, length=0.505, radius=1, shapeName=dyn_chain_l_3_capsule, rotation=<rot 0 -90.3797 92.9326>}]
 dyn_chain_r_4 [{
   shapeType=Capsule
   length=2.2052
   radius=1.3035
   shapeName=dyn_chain_l_4_capsule
   center=<0.6639 0.0056 -0.0437>
   rotation=<rot 0 -89.519 98.665>
  }]
 dyn_chain_r_5 [{
   shapeType=Capsule
   length=1.5226
   radius=1.3698
   shapeName=dyn_chain_l_5_capsule
   center=<0.0023 0 0.1453>
   rotation=<rot 0 -88.7777 90.8983>
  }]
 dyn_chain_r_6 [{
   shapeType=Capsule
   length=2.505
   radius=1.5
   shapeName=dyn_chain_l_6_capsule
   center=<1.1538 0.0261 0.0266>
   rotation=<rot 0 -88.704 92.8314>
  }]
 pack [{shapeType=Capsule, length=16.3613, radius=12.2328, center=<0.0021 -4.1127 -3.1298>, rotation=<rot 2.7386 -90 -0.0181>}]
}
constraints [
 {bone1Name=base_chain_l, bone2Name=Sword_handle_l, swing1LimitDegrees=10, swing2LimitDegrees=10, twistLimitDegrees=3}
 {bone1Name=dyn_chain_l_1, bone2Name=base_chain_l, swing1LimitDegrees=20, swing2LimitDegrees=25, twistLimitDegrees=3}
 {bone1Name=dyn_chain_l_2, bone2Name=dyn_chain_l_1, swing1LimitDegrees=10, swing2LimitDegrees=20, twistLimitDegrees=3}
 {bone1Name=dyn_chain_l_3, bone2Name=dyn_chain_l_2, swing1LimitDegrees=13, swing2LimitDegrees=22, twistLimitDegrees=3}
 {bone1Name=dyn_chain_l_4, bone2Name=dyn_chain_l_3, swing1LimitDegrees=10, swing2LimitDegrees=20, twistLimitDegrees=3}
 {bone1Name=dyn_chain_l_5, bone2Name=dyn_chain_l_4, swing1LimitDegrees=10, swing2LimitDegrees=20, twistLimitDegrees=3}
 {bone1Name=dyn_chain_l_6, bone2Name=dyn_chain_l_5, swing1LimitDegrees=10, swing2LimitDegrees=10, twistLimitDegrees=3}
 {
  bone1Name=Sword_handle_l
  bone2Name=pack
  swing1LimitDegrees=45
  swing1Motion=Free
  swing2LimitDegrees=45
  swing2Motion=Free
  twistLimitDegrees=45
  twistMotion=Free
 }
 {
  bone1Name=Sword_handle_r
  bone2Name=pack
  swing1LimitDegrees=45
  swing1Motion=Free
  swing2LimitDegrees=45
  swing2Motion=Free
  twistLimitDegrees=45
  twistMotion=Free
 }
 {bone1Name=base_chain_r, bone2Name=Sword_handle_r, swing1LimitDegrees=10, swing2LimitDegrees=10, twistLimitDegrees=3}
 {bone1Name=dyn_chain_r_1, bone2Name=base_chain_r, swing1LimitDegrees=20, swing2LimitDegrees=25, twistLimitDegrees=3}
 {bone1Name=dyn_chain_r_2, bone2Name=dyn_chain_r_1, swing1LimitDegrees=10, swing2LimitDegrees=10, twistLimitDegrees=3}
 {bone1Name=dyn_chain_r_3, bone2Name=dyn_chain_r_2, swing1LimitDegrees=10, swing2LimitDegrees=21, twistLimitDegrees=5}
 {bone1Name=dyn_chain_r_4, bone2Name=dyn_chain_r_3, swing1LimitDegrees=10, swing2LimitDegrees=21, twistLimitDegrees=5}
 {bone1Name=dyn_chain_r_5, bone2Name=dyn_chain_r_4, swing1LimitDegrees=10, swing2LimitDegrees=21, twistLimitDegrees=5}
 {bone1Name=dyn_chain_r_6, bone2Name=dyn_chain_r_5, swing1LimitDegrees=10, swing2LimitDegrees=21, twistLimitDegrees=5}
]
