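# Anim Blueprint for the KelpLinenMagnesium backpack cosmetic, posing the pack against the character's body and adding secondary motion.
# Body-type offsets and a Control Rig, fed body socket transforms and crouch, DBNO and parachute states, are weighted by Armed.
# Fixed bone offsets follow, then separately tuned RigidBody simulations for gameplay and front end.
# Weapon changes detect the tagged pickaxe, toggling hide-element material parameters and lowering Armed while it is equipped.
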AnimBlueprint /BRCosmetics/Accessories/FORT_Backpacks/Backpack_M_MED_KelpLinenMagnesium/Meshes/M_MED_KelpLinenMagnesium_Pack_AnimBP
target_skeleton=&/M_MED_KelpLinenMagnesium_Pack_Skeleton
graphs [
 AnimGraph {nodes [
   n1 = "Transform (Modify) Bone - Bone: pack"(
    ComponentPose=Local To Component(LocalPose=Input Pose())
    Alpha=Get Armed() -> (Armed:double)
    Translation=Get backpack_translation()
    Rotation=Get backpack_rotate()
    Scale=Get Backpack Scale() -> (Backpack Scale:Vector)
   ) {
    rotationMode=BMM_Additive
    rotationSpace=BCS_BoneSpace
    scaleMode=BMM_Additive
    scaleSpace=BCS_BoneSpace
    translationMode=BMM_Additive
    translationSpace=BCS_BoneSpace
    alphaBoolBlend {blendInTime=0.2, blendOption=Sinusoidal, blendOutTime=0.2}
    boneToModify {boneName=pack}
   }
   n2 = Control Rig!(
    Source=Component To Local(ComponentPose=@n1)
    Alpha=Get Armed()
    pelvis:Transform=Get Pelvis() -> (Pelvis:Transform)
    spine_05:Transform=Get Spine_05() -> (Spine_05:Transform)
    spine_03:Transform=Get spine_03() -> (spine_03:Transform)
    spine_04:Transform=Get spine_04() -> (spine_04:Transform)
    root:Transform=Get root() -> (root:Transform)
    DBNO:bool=Get bIsDBNO()
    head:Transform=Get head() -> (head:Transform)
    is parachute open:bool=Get bIsParachuteOpened()
    spine_01:Transform=Get spine_01() -> (spine_01:Transform)
    is Crouch:bool=Get bIsCrouching()
    neck_01:Transform=Get Neck_01() -> (Neck_01:Transform)
   ) -> (Pose) {
    alphaBoolBlend {blendOption=Linear}
    controlRigAssetReference {blueprintRigClass=&/M_MED_KelpLinenMagnesium_Pack_CtrlRig.M_MED_KelpLinenMagnesium_Pack_CtrlRig_C}
   }
   n3 = "Transform (Modify) Bone - Bone: pack"(ComponentPose=Local To Component(LocalPose=@n2), Translation="0,-1,0", Rotation="0,0,2") {
    rotationMode=BMM_Additive
    rotationSpace=BCS_BoneSpace
    translationMode=BMM_Additive
    translationSpace=BCS_BoneSpace
    alphaBoolBlend {blendOption=Linear}
    boneToModify {boneName=pack}
    rotation {roll=2}
    translation {y=-1}
   }
   n4 = "Transform (Modify) Bone - Bone: pack"(ComponentPose=@n3, Translation="0,0,-4", Rotation="0,0,0", Scale="0.9,0.9,0.9") {
    rotationMode=BMM_Additive
    scaleMode=BMM_Additive
    translationMode=BMM_Additive
    alphaBoolBlend {blendOption=Linear}
    boneToModify {boneName=pack}
    scale=<0.9 0.9 0.9>
    translation {z=-4}
   }
   n5 = RigidBody!(ComponentPose=@n4, bAlphaBoolEnabled:bool=NOT Boolean(A=Get bIsInFrontEnd()), OverrideWorldGravity:Vector="0,0,-980", ExternalForce) -> (Pose) {
    alphaInputType=Bool
    bOverrideWorldGravity=true
    overridePhysicsAsset=&/M_MED_KelpLinenMagnesium_Pack_Physics
    simulationSpace=BaseBoneSpace
    alphaBoolBlend {blendOption=Linear}
    baseBoneRef {boneName=pack}
    componentAppliedLinearAccClamp=<3000 3500 2500>
    componentLinearAccScale=<0.4 0.4 0.5>
    componentLinearVelScale=<0.4 0.4 0.5>
    overrideWorldGravity {z=-980}
    simSpaceSettings {maxAngularAcceleration=130, maxAngularVelocity=3, maxLinearAcceleration=130, maxLinearVelocity=150, velocityScaleZ=0, worldAlpha=0.6}
   }
   n6 = RigidBody(ComponentPose=@n5, bAlphaBoolEnabled=Get bIsInFrontEnd(), OverrideWorldGravity="0,0,-980") {
    alphaInputType=Bool
    bOverrideWorldGravity=true
    overridePhysicsAsset=&M_MED_KelpLinenMagnesium_Pack_Physics
    simulationSpace=BaseBoneSpace
    alphaBoolBlend {blendOption=Linear}
    baseBoneRef {boneName=pack}
    componentLinearAccScale=<1 1 1>
    componentLinearVelScale=<1 1 1>
    overrideWorldGravity {z=-980}
    simSpaceSettings {maxAngularAcceleration=300, maxAngularVelocity=300, maxLinearAcceleration=300, maxLinearVelocity=300, velocityScaleZ=0, worldAlpha=1}
   }
   Output Pose(Result=Component To Local(ComponentPose=@n6)) {layerGroup=None}
  ]}
 EventGraph {nodes [
   Event BlueprintUpdateAnimation()
   > n1 = Sequence!(execute) -> (then_0, then_1, then_2:Exec, then_3:Exec)
   > n2 = Gravity Override State Function DEPRECATED(
    CharacterPartAnimInstance=Self-Reference()
    GameTimeInSeconds=GetGameTimeInSeconds()
    GravityOverrideParams=Get Gravity Override Params()
    DebugJointName=pack
   )
   > Set Gravity Override(Gravity Override=@n2)
   > n3 = Branch(Condition=Get isF_Med() -> (isF_Med:bool))
   > Set backpack_rotate(backpack_rotate="0,0,4")
   > Set backpack_translation(backpack_translation="0,-0.5,-3")
   > Set Backpack Scale(Backpack Scale:Vector="0.9,0.9,0.9") -> (Output_Get:Vector)
   Set backpack_rotate(execute=@n3.else, backpack_rotate="0,0,10")
   > Set backpack_translation(backpack_translation="0,-1,-3")
   > Set Backpack Scale(Backpack Scale="1,1,1")
   Reroute Node(InputPin=@n1.then_3)
   n4 = Get bIsParachuteOpened()
   n5 = Branch(execute=@n1.then_2, Condition=AND Boolean(A=Get bIsSkyDiving(), B=NOT Boolean(A=@n4)))
   > Set SkyDiveNotPrachute(SkyDiveNotPrachute:double=1) -> (Output_Get:double)
   Set SkyDiveNotPrachute(execute=@n5.else, SkyDiveNotPrachute=0)
   n6 = Get Body() -> (Body:Skeletal Mesh Component Budgeted Object Reference)
   Set Pelvis(execute=@n1.then_1, Pelvis:Transform=GetSocketTransform(self=@n6, InSocketName=pelvis, TransformSpace=RTS_World)) -> (Output_Get:Transform)
   > Set Spine_05(Spine_05:Transform=GetSocketTransform(self=@n6, InSocketName=spine_05, TransformSpace=RTS_World)) -> (Output_Get:Transform)
   > Set spine_04(spine_04:Transform=GetSocketTransform(self=@n6, InSocketName=spine_04, TransformSpace=RTS_World)) -> (Output_Get:Transform)
   > Set spine_03(spine_03:Transform=GetSocketTransform(self=@n6, InSocketName=spine_03, TransformSpace=RTS_World)) -> (Output_Get:Transform)
   > Set root(root:Transform=GetSocketTransform(self=@n6, InSocketName=root, TransformSpace=RTS_World)) -> (Output_Get:Transform)
   > Set head(head:Transform=GetSocketTransform(self=@n6, InSocketName=head, TransformSpace=RTS_World)) -> (Output_Get:Transform)
   > Set spine_01(spine_01:Transform=GetSocketTransform(self=@n6, InSocketName=spine_01, TransformSpace=RTS_World)) -> (Output_Get:Transform)
   > Set Neck_01(Neck_01:Transform=GetSocketTransform(self=@n6, InSocketName=neck_01, TransformSpace=RTS_World)) -> (Output_Get:Transform)
   > n7 = Branch(Condition="float > float"(A=Get PawnVelocityZ(), B=1111))
   > Set Batoot_aplha(Batoot_aplha:double=1) -> (Output_Get:double)
   Set Batoot_aplha(execute=@n7.else, Batoot_aplha=0)
   n8 = Event OnCharacterCustomizationLoaded()
   > Set Backpack(Backpack:Skeletal Mesh Component Object Reference=@n8.CustomBaseMesh) -> (Output_Get:Skeletal Mesh Component Object Reference)
   > Set Body(Body:Skeletal Mesh Component Budgeted Object Reference=Get Array Item with EFort Custom Part Type(Array=@n8.CustomMeshParts)) -> (Output_Get:Skeletal Mesh Component Budgeted Object Reference)
   n9 = "Get (a copy)"(Array:Array of Custom Character Part Object References=@n8.CustomParts, Dimension 1=1) -> (Output:Custom Character Part Object Reference)
   n10 = Get GenderPermitted!(self:Custom Character Part Object Reference=@n9) -> (GenderPermitted:EFortCustomGender Enum=Invalid)
   n11 = Get BodyTypesPermitted!(self:Custom Character Part Object Reference=@n9) -> (BodyTypesPermitted:EFortCustomBodyType Enum=NONE)
   > Set isF_Med(isF_Med:bool=AND Boolean(
     A="Equal (Enum)"(A:EFortCustomBodyType Enum=@n11, B:EFortCustomBodyType Enum=Medium)
     B="Equal (Enum)"(A:EFortCustomGender Enum=@n10, B:EFortCustomGender Enum=Female)
    )) -> (Output_Get:bool)
   Event OnWeaponChanged()
   > Is Valid(InputObject=Get Backpack() -> (Backpack:Skeletal Mesh Component Object Reference))
   > n12 = Get Dynamic Material Instance(SkeletalMeshComponent=GetPartSkeletalMeshComponent())
   > n13 = Branch(Condition=PickaxeItemDefinitionHasAnyMetaTag(Tags=Make Array("[0]"=(TagName="Cosmetics.Metatags.KelpLinenMagnesium")), bRequiresEquipped=true))
   > SetScalarParameterValue(self=@n12, ParameterName=Hide Element 0_5, Value=0)
   > SetScalarParameterValue(self=@n12, ParameterName=Hide Element 01, Value=1)
   n14 = GetPartSkeletalMeshComponent()
   > SetScalarParameterValueOnOverlayMaterial!(execute, MeshComponent:Mesh Component Object Reference=@n14, ParameterName:Name=Hide Element 01, ParameterValue:float=1) -> (then)
   > Set Armed(Armed:double=0.15) -> (Output_Get:double)
   Delay(execute=@n13.else, Duration=0.1)
   > SetScalarParameterValue(self=@n12, ParameterName=Hide Element 01, Value=0)
   > SetScalarParameterValue(self=@n12, ParameterName=Hide Element 0_5, Value=1)
   > SetScalarParameterValueOnOverlayMaterial(MeshComponent=@n14, ParameterName=Hide Element 01, ParameterValue=0)
   > Set Armed(Armed=1)
  ]}
]
variables [
 Armed {type=number}
 Backpack {type=SkeletalMeshComponent}
 Backpack Scale {type=Vector, default=<0.9 0.9 0.8>}
 backpack_rotate {type=Rotator}
 backpack_translation {type=Vector}
 Batoot_aplha {type=number}
 Body {type=SkeletalMeshComponentBudgeted}
 Custom Base Mesh {type=SkeletalMeshComponent}
 Gravity Override {type=Vector}
 Gravity Override Params {type=GravityOverrideParamsStruct, default {dBNO {jointName=root, gravityOverride=<0 0 -980>}, front_End {jointName=pack}}}
 head {type=Transform, default {location=<0 0 0>, rotation=<rot 0 0 0>, scale=<1 1 1>}}
 isF_Med {type=boolean}
 Neck_01 {type=Transform, default {location=<0 0 0>, rotation=<rot 0 0 0>, scale=<1 1 1>}}
 Pelvis {type=Transform, default {location=<0 0 0>, rotation=<rot 0 0 0>, scale=<1 1 1>}}
 root {type=Transform, default {location=<0 0 0>, rotation=<rot 0 0 0>, scale=<1 1 1>}}
 SkyDiveNotPrachute {type=number}
 spine_01 {type=Transform, default {location=<0 0 0>, rotation=<rot 0 0 0>, scale=<1 1 1>}}
 spine_03 {type=Transform, default {location=<0 0 0>, rotation=<rot 0 0 0>, scale=<1 1 1>}}
 spine_04 {type=Transform, default {location=<0 0 0>, rotation=<rot 0 0 0>, scale=<1 1 1>}}
 Spine_05 {type=Transform, default {location=<0 0 0>, rotation=<rot 0 0 0>, scale=<1 1 1>}}
]
